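# Masked, two-sided lit material for mesh particles that renders lava popping, reusing water-splash textures.
# The Popped dynamic parameter blends from a solid mask and smooth normal to the splash mask and normal map.
# WPOSpread, a WPO vector and red vertex color build a transformed offset for spreading the mesh.
# A Fresnel blend of two lava colors, tinted by particle color, feeds HDR remapping and emissive adjustment functions.
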Material /Game/Athena/Environments/Materials/Lava/M_Lava_Pop
graph {
 properties {bUsedWithMeshParticles=true, blendMode=BLEND_Masked, pixelDepthOffsetMode=PDOM_Legacy, twoSided=true, usedShadingModels=Default Lit}
 nodes [
  n1 = DynamicParameter() {defaultValue=<rgba 0.2 0 1 1>, paramNames [LavaIntensity, Popped, WPOSpread, Param4]}
  LinearInterpolate(
   A=Constant() {r=1}
   B=TextureSample() {samplerType=SAMPLERTYPE_LinearColor, texture=/Game/Effects/Fort_Effects/Textures/Water/T_waterSplash}.G
   Alpha=@n1.Popped
  )
  LinearInterpolate(
   A=VectorParameter() {parameterName=SmoothNormal, defaultValue=<rgba 0 0 1 0>}.RGB
   B=TextureSample() {samplerType=SAMPLERTYPE_Normal, texture=/Game/Effects/Fort_Effects/Textures/Water/T_waterSplash_N}.RGB
   Alpha=@n1.Popped
  )
  n2 = Multiply(
   A=LinearInterpolate(
    A=VectorParameter() {parameterName=BaseColor0, defaultValue=<rgba 0.1736 0.0342 0 0>}.RGB
    B=VectorParameter() {parameterName=BaseColor1, defaultValue=<rgba 1 0.2057 0.0187 0>}.RGB
    Alpha=Fresnel()
   )
   B=ParticleColor().RGB
  )
  MaterialFunctionCall(Emissive=MaterialFunctionCall(LDR Input=Clamp(None=@n2), HDR Tint=@n2) {materialFunction=&DeriveHDRfromLDR}.RemappedRange) {materialFunction=&MF_FX_EmissiveAdjustments}
  ScalarParameter() {defaultValue=0.1, parameterName=Roughness}
  n3 = VectorParameter() {parameterName=WPO, defaultValue=<rgba 1 1 1 0>}
  Transform(None=Multiply(A=Multiply(A=@n1.WPOSpread, B=@n3.RGB), B=VertexColor().R))
  VectorParameter() {parameterName=BaseColor, defaultValue=<rgba 0.6007 0.1235 0.0113 0>}
 ]
}
